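# Sound cue for the firing sound of the poison dart trap, in the trap placement sound class with shared trap attenuation settings.
# A mixer layers one of two random fire shots, passed through a modulator, with a dart flyby played after a short random delay.
# The flyby branch overrides attenuation with its own falloff.
# Overridden concurrency prevents new instances once the limit is reached.
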SoundCue /Game/Sounds/Fort_Traps/PoisonDart_Trap/PoisonDart_Trap_Fire_Cue
root=Mixer_0
nodes [
 {id=Attenuation_0, children [Modulator_3]}
 {
  id=Attenuation_1
  children [Delay_0]
  properties {bOverrideAttenuation=true, attenuationOverrides {falloffDistance=2048, attenuationShapeExtents {x=512}}}
 }
 {id=Delay_0, children [WavePlayer_11], properties {delayMax=0.3, delayMin=0.1}}
 {id=Mixer_0, children [Attenuation_1, Attenuation_0], properties {inputVolume [1,1]}}
 {id=Modulator_3, children [Random_0]}
 {id=Random_0, children [WavePlayer_6,WavePlayer_7], properties {weights [1,1]}}
 {id=WavePlayer_11, properties {soundWaveAssetPtr=&/PoisonDartTrap_Flyby_01}}
 {id=WavePlayer_6, properties {soundWaveAssetPtr=&/PoisonDartTrap_Fire_02}}
 {id=WavePlayer_7, properties {soundWaveAssetPtr=&/PoisonDartTrap_Fire_01}}
]
properties {
 attenuationSettings=&Fort_Traps_Att
 bOverrideConcurrency=true
 maxDistance=2097152
 soundClassObject=/Game/Sounds/FortSoundClassesAndModes/Fort_Traps_Placement
 volumeMultiplier=1
 attenuationOverrides {bAttenuateWithLPF=true, falloffDistance=3100, lPFRadiusMin=2500, attenuationShapeExtents {x=900}}
 concurrencyOverrides {resolutionRule=PreventNew}
}
