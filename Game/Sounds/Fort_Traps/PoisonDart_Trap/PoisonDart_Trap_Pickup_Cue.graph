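# Sound cue for the pickup sound of the Poison Dart trap.
# A single wave player feeds the PoisonDartTrap_Pickup_01 sound through a Modulator at default settings.
# It plays at reduced volume in the gadgets and items sound class, using weapon-drop attenuation for distance falloff.
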SoundCue /Game/Sounds/Fort_Traps/PoisonDart_Trap/PoisonDart_Trap_Pickup_Cue
root=Modulator_0
nodes [
 {id=Modulator_0, children [WavePlayer_1]}
 {id=WavePlayer_1, properties {soundWaveAssetPtr=&/PoisonDartTrap_Pickup_01}}
]
properties {
 attenuationSettings=&Fort_Weapon_Drop_Att
 maxDistance=3000
 soundClassObject=&Fort_Gadgets_Items
 volumeMultiplier=0.4
}
