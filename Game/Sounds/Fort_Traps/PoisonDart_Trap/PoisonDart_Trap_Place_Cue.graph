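# Sound cue for placing the Poison Dart trap.
# A random node picks one of two placement wave variants with equal weight and feeds it through a modulator at the root.
# It plays in the trap placement sound class with trap attenuation settings.
# A concurrency override prevents new instances from starting once the limit is reached.
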SoundCue /Game/Sounds/Fort_Traps/PoisonDart_Trap/PoisonDart_Trap_Place_Cue
root=Modulator_3
nodes [
 {id=Modulator_3, children [Random_0]}
 {id=Random_0, children [WavePlayer_5,WavePlayer_6], properties {weights [1,1]}}
 {id=WavePlayer_5, properties {soundWaveAssetPtr=&/PoisonDartTrap_Place_01}}
 {id=WavePlayer_6, properties {soundWaveAssetPtr=&/PoisonDartTrap_Place_02}}
]
properties {
 attenuationSettings=&Fort_Traps_Att
 bOverrideConcurrency=true
 maxDistance=6000
 soundClassObject=/Game/Sounds/FortSoundClassesAndModes/Fort_Traps_Placement
 volumeMultiplier=0.9
 attenuationOverrides {falloffDistance=2400, attenuationShapeExtents {x=600}}
 concurrencyOverrides {resolutionRule=PreventNew}
}
