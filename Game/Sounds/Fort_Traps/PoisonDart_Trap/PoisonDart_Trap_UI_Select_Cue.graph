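# Sound cue for the UI select sound of the Poison Dart trap, reusing the trap's pickup sound wave.
# A single wave player feeds a Modulator left at its default settings.
# The cue plays at reduced volume through the non-critical in-game UI sound class.
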SoundCue /Game/Sounds/Fort_Traps/PoisonDart_Trap/PoisonDart_Trap_UI_Select_Cue
root=Modulator_0
nodes [
 {id=Modulator_0, children [WavePlayer_1]}
 {id=WavePlayer_1, properties {soundWaveAssetPtr=&/PoisonDartTrap_Pickup_01}}
]
properties {maxDistance=2097152, soundClassObject=&Fort_UI_InGame_NonCritical, volumeMultiplier=0.51}
