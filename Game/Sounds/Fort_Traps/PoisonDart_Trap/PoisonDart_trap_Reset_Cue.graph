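# Sound cue that plays the reset sound for the Poison Dart trap.
# A Random node picks one of two equally weighted reset waves and feeds a Modulator at the root, which is left at default settings.
# The cue plays at reduced volume in the Fort_Traps sound class with the shared Fort_Traps_Att attenuation.
# Its own concurrency override prevents new instances once the limit is reached.
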SoundCue /Game/Sounds/Fort_Traps/PoisonDart_Trap/PoisonDart_trap_Reset_Cue.PoisonDart_Trap_Reset_Cue
root=Modulator_8
nodes [
 {id=Modulator_8, children [Random_0]}
 {id=Random_0, children [WavePlayer_2,WavePlayer_3], properties {weights [1,1]}}
 {id=WavePlayer_2, properties {soundWaveAssetPtr=&/PoisonDartTrap_Reset_01}}
 {id=WavePlayer_3, properties {soundWaveAssetPtr=&/PoisonDartTrap_Reset_02}}
]
properties {
 attenuationSettings=&Fort_Traps_Att
 bOverrideConcurrency=true
 maxDistance=6000
 soundClassObject=&Fort_Traps
 volumeMultiplier=0.5
 attenuationOverrides {falloffDistance=2400, attenuationShapeExtents {x=600}}
 concurrencyOverrides {resolutionRule=PreventNew}
}
